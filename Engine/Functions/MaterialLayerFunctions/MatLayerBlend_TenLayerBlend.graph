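# Library material function that composites a Background and up to ten material-attribute layers into one Blended Material output.
# Ten chained MatLayerBlend_Standard calls apply Layer1 through Layer10 in order, each by its own scalar mask.
# The final result passes through MatLayerBlend_BakedNormal with an optional baked normal that defaults to flat.
# All inputs fall back to preview defaults, so unconnected layers are expected to be optimized out by the compiler.
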MaterialFunction /Engine/Functions/MaterialLayerFunctions/MatLayerBlend_TenLayerBlend
graph {
 properties {
  bExposeToLibrary=true
  description="Fixed function blending of 10 layers.  Only uses the layers hooked up.  Compiler should optimize out unused layers."
  libraryCategoriesText [Misc, MaterialLayerBlend]
 }
 nodes [
  n1 = MakeMaterialAttributes!(
   BaseColor
   Metallic
   Specular
   Roughness
   Anisotropy
   EmissiveColor
   Opacity
   OpacityMask
   Normal
   Tangent
   WorldPositionOffset
   SubsurfaceColor
   ClearCoat
   ClearCoatRoughness
   AmbientOcclusion
   Refraction
   CustomizedUVs_0
   CustomizedUVs_1
   CustomizedUVs_2
   CustomizedUVs_3
   CustomizedUVs_4
   CustomizedUVs_5
   CustomizedUVs_6
   CustomizedUVs_7
   PixelDepthOffset
   ShadingModel
   Displacement
  )
  n2 = FunctionInput(Preview=@n1) {bUsePreviewValueAsDefault=true, inputName=Background, inputType=FunctionInput_MaterialAttributes, sortPriority=30}
  n3 = FunctionInput(Preview=@n1) {bUsePreviewValueAsDefault=true, inputName=Layer1, inputType=FunctionInput_MaterialAttributes, sortPriority=27}
  n4 = FunctionInput() {bUsePreviewValueAsDefault=true, inputName=Layer1_Mask, inputType=FunctionInput_Scalar, sortPriority=28}
  n5 = MaterialFunctionCall!(
   Base Material=MaterialFunctionCall(
    Base Material=MaterialFunctionCall(Base Material=@n2, Top Material=@n3, Alpha=@n4) {materialFunction=&/MatLayerBlend_Standard}.Blended Material
    Top Material=FunctionInput(Preview=@n1) {bUsePreviewValueAsDefault=true, inputName=Layer2, inputType=FunctionInput_MaterialAttributes, sortPriority=24}
    Alpha=FunctionInput() {bUsePreviewValueAsDefault=true, inputName=Layer2_Mask, inputType=FunctionInput_Scalar, sortPriority=25}
   ) {materialFunction=&MatLayerBlend_Standard}.Blended Material
   Top Material=FunctionInput(Preview=@n1) {bUsePreviewValueAsDefault=true, inputName=Layer3, inputType=FunctionInput_MaterialAttributes, sortPriority=21}
   Alpha=FunctionInput() {bUsePreviewValueAsDefault=true, inputName=Layer3_Mask, inputType=FunctionInput_Scalar, sortPriority=22}
  ) {materialFunction=&MatLayerBlend_Standard}
  n6 = FunctionInput(Preview=@n1) {bUsePreviewValueAsDefault=true, inputName=Layer4, inputType=FunctionInput_MaterialAttributes, sortPriority=18}
  n7 = FunctionInput() {bUsePreviewValueAsDefault=true, inputName=Layer4_Mask, inputType=FunctionInput_Scalar, sortPriority=19}
  n8 = MaterialFunctionCall(
   Base Material=MaterialFunctionCall(
    Base Material=MaterialFunctionCall(Base Material=@n5.Blended Material, Top Material=@n6, Alpha=@n7) {materialFunction=&MatLayerBlend_Standard}.Blended Material
    Top Material=FunctionInput(Preview=@n1) {bUsePreviewValueAsDefault=true, inputName=Layer5, inputType=FunctionInput_MaterialAttributes, sortPriority=15}
    Alpha=FunctionInput() {bUsePreviewValueAsDefault=true, inputName=Layer5_Mask, inputType=FunctionInput_Scalar, sortPriority=16}
   ) {materialFunction=&MatLayerBlend_Standard}.Blended Material
   Top Material=FunctionInput(Preview=@n1) {bUsePreviewValueAsDefault=true, inputName=Layer6, inputType=FunctionInput_MaterialAttributes, sortPriority=12}
   Alpha=FunctionInput() {bUsePreviewValueAsDefault=true, inputName="Layer6_Mask\n", inputType=FunctionInput_Scalar, sortPriority=13}
  ) {materialFunction=&MatLayerBlend_Standard}
  n9 = FunctionInput(Preview=@n1) {bUsePreviewValueAsDefault=true, inputName=Layer7, inputType=FunctionInput_MaterialAttributes, sortPriority=9}
  n10 = FunctionInput() {bUsePreviewValueAsDefault=true, inputName=Layer7_Mask, inputType=FunctionInput_Scalar, sortPriority=10}
  n11 = MaterialFunctionCall(
   Base Material=MaterialFunctionCall(
    Base Material=MaterialFunctionCall(Base Material=@n8.Blended Material, Top Material=@n9, Alpha=@n10) {materialFunction=&MatLayerBlend_Standard}.Blended Material
    Top Material=FunctionInput(Preview=@n1) {bUsePreviewValueAsDefault=true, inputName=Layer8, inputType=FunctionInput_MaterialAttributes, sortPriority=6}
    Alpha=FunctionInput() {bUsePreviewValueAsDefault=true, inputName="Layer8_Mask\n", inputType=FunctionInput_Scalar, sortPriority=7}
   ) {materialFunction=&MatLayerBlend_Standard}.Blended Material
   Top Material=FunctionInput(Preview=@n1) {bUsePreviewValueAsDefault=true, inputName=Layer9, inputType=FunctionInput_MaterialAttributes, sortPriority=3}
   Alpha=FunctionInput() {bUsePreviewValueAsDefault=true, inputName=Layer9_Mask, inputType=FunctionInput_Scalar, sortPriority=4}
  ) {materialFunction=&MatLayerBlend_Standard}
  n12 = FunctionInput(Preview=@n1) {bUsePreviewValueAsDefault=true, inputName=Layer10, inputType=FunctionInput_MaterialAttributes}
  n13 = FunctionInput() {bUsePreviewValueAsDefault=true, inputName=Layer10_Mask, inputType=FunctionInput_Scalar, sortPriority=1}
  n14 = FunctionInput() {bUsePreviewValueAsDefault=true, inputName="<BakedNormal>", sortPriority=31, previewValue=<0 0 1 1>}
  FunctionOutput(None=MaterialFunctionCall!(
    Material=MaterialFunctionCall(Base Material=@n11.Blended Material, Top Material=@n12, Alpha=@n13) {materialFunction=&MatLayerBlend_Standard}.Blended Material
    Normal=@n14
   ) {materialFunction=&/MatLayerBlend_BakedNormal}.Blended Material) {outputName=Blended Material}
 ]
}
