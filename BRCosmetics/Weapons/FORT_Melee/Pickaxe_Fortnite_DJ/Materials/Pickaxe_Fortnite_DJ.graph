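# Skinned-mesh material for the Fortnite DJ pickaxe, blending animated emissive disco lights over a textured base.
# Time-driven sine and cosine curves reshape packed-mask values, which a hard threshold turns into blinking lights, with optional UV scrolling for falling motion.
# A second thresholded layer pans a stripe-tiled mask.
# Both are tinted by the emissive texture, boosted through DeriveHDRfromLDR, then mask-blended via TenLayerBlend and passed through accessory TechArt overrides.
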Material /BRCosmetics/Weapons/FORT_Melee/Pickaxe_Fortnite_DJ/Materials/Pickaxe_Fortnite_DJ
graph {
 properties {
  bUseMaterialAttributes=true
  bUsedWithMeshDeformer=true
  bUsedWithSkeletalMesh=true
  pixelDepthOffsetMode=PDOM_Legacy
  usedShadingModels=Default Lit
 }
 nodes [
  n1 = TextureSample() {samplerType=SAMPLERTYPE_LinearColor, texture=/BRCosmetics/Weapons/FORT_Melee/Pickaxe_Fortnite_DJ/Textures/T_Pickaxe_Fortnite_DJ_E}
  n2 = Multiply(A=TextureCoordinate(), B=ScalarParameter() {defaultValue=1, group=DiscoLights, parameterName=DiscoLights_Tiling})
  n3 = Multiply(
   A=Time()
   B=AppendVector(
    A=ScalarParameter() {group=DiscoLights, parameterName=FallSpeed_X}
    B=ScalarParameter() {defaultValue=2, group=DiscoLights, parameterName=FallSpeed_Y}
   )
  )
  n4 = TextureSample(UVs=LinearInterpolate(A=@n2, B=Add(A=@n2, B=@n3), Alpha=ScalarParameter() {group=DiscoLights, parameterName=isFalling})) {samplerType=SAMPLERTYPE_Masks, texture=/BRCosmetics/Weapons/FORT_Melee/Pickaxe_Fortnite_DJ/Textures/T_Pickaxe_Fortnite_DJ_M}
  n5 = ScalarParameter() {desc=New, group=DiscoLights, parameterName=ThresholdOffset}
  n6 = LinearInterpolate(A=Add(A=@n4.B, B=@n5) {constB=0.6, desc=New}, B=Add(A=@n4.R, B=@n5) {constB=0.6, desc=New})
  n7 = Multiply(A=Time(), B=ScalarParameter() {defaultValue=0.8, group=DiscoLights, parameterName=DiscoLights_Speed})
  n8 = Multiply(A=@n7) {constB=0.25}
  n9 = Multiply(A=@n7) {constB=0.5}
  n10 = LinearInterpolate(
   A=Abs(None=Cosine(None=@n7))
   B=Abs(None=Cosine(None=@n8))
   Alpha=Abs(None=MaterialFunctionCall(Value=@n9) {materialFunction=&LinearSine}.Linear Sine)
  )
  n11 = Abs(None=Frac(None=Multiply(A=@n7) {constB=0.125}))
  n12 = Power(Base=@n6, Exp=LinearInterpolate(A=@n10, B=@n10, Alpha=Clamp(None=@n11)) {constA=2, constB=2.4})
  n13 = MaterialFunctionCall(Value=Multiply(A=@n7) {constB=0.5}) {materialFunction=&LinearSine}
  n14 = Power(Base=LinearInterpolate(A=Power(Base=@n6) {constExponent=4}, B=Clamp(None=@n12), Alpha=Abs(None=@n13.Linear Sine))) {constExponent=4}
  n15 = ScalarParameter() {defaultValue=0.047, group=DiscoLights, parameterName=DiscoThreshold_Min}
  n16 = Constant() {r=1}
  n17 = Constant()
  n18 = MaterialFunctionCall(
   LDR Input=Multiply(A=@n1.RGB, B=If(A=@n14, B=@n15, "A > B"=@n16, "A < B"=@n17))
   HDR Intensity Multiplier=ScalarParameter() {defaultValue=8, group=DiscoLights, parameterName=HDRIntensity}
   Derive HDR Power=ScalarParameter() {defaultValue=1.5, group=DiscoLights, parameterName=HDRPower}
   Desaturate Luminance=ScalarParameter() {defaultValue=1, group=DiscoLights, parameterName=LuminanceDesat}
  ) {materialFunction=&DeriveHDRfromLDR}
  n19 = ScalarParameter() {defaultValue=1, group=DiscoLights, parameterName=DiscoLights_EmissiveBrightness}
  n20 = Panner(Coordinate=TextureCoordinate() {uTiling=5}) {speedX=-1}
  n21 = Multiply(
   A=@n4.A
   B=Multiply(A=TextureSample(UVs=@n20) {texture=/BRCosmetics/Weapons/FORT_Melee/Pickaxe_Fortnite_DJ/Textures/T_Pickaxe_Fortnite_DJ_M1}.R, B=@n4.A)
  )
  n22 = ScalarParameter() {defaultValue=0.0533, parameterName=test}
  n23 = Constant() {r=1}
  n24 = Constant()
  n25 = Multiply(A=Clamp(None=If(A=@n21, B=@n22, "A > B"=@n23, "A < B"=@n24)), B=@n1.RGB)
  n26 = ScalarParameter() {defaultValue=3, group=DiscoLights, parameterName=HDRIntensity_2}
  n27 = ScalarParameter() {defaultValue=1.5, group=DiscoLights, parameterName=HDRPower_2}
  n28 = ScalarParameter() {defaultValue=1, group=DiscoLights, parameterName=LuminanceDesat_2}
  n29 = Add(
   A=Multiply(B=OneMinus(None=@n4.R))
   B=Add(
    A=Multiply(A=@n18.RemappedRange, B=@n19)
    B=MaterialFunctionCall(LDR Input=@n25, HDR Intensity Multiplier=@n26, Derive HDR Power=@n27, Desaturate Luminance=@n28) {materialFunction=&DeriveHDRfromLDR}.RemappedRange
   )
  )
  n30 = Add(A=@n4.A, B=@n4.A)
  n31 = TextureSample() {texture=/BRCosmetics/Weapons/FORT_Melee/Pickaxe_Fortnite_DJ/Textures/T_Pickaxe_Fortnite_DJ_D}
  n32 = TextureSample() {samplerType=SAMPLERTYPE_Masks, texture=/BRCosmetics/Weapons/FORT_Melee/Pickaxe_Fortnite_DJ/Textures/T_Pickaxe_Fortnite_DJ_S}
  MaterialFunctionCall(In=MaterialFunctionCall!(
    Layer10
    Layer10_Mask
    Layer9
    Layer9_Mask
    Layer8
    "Layer8_Mask\n"
    Layer7
    Layer7_Mask
    Layer6
    "Layer6_Mask\n"
    Layer5
    Layer5_Mask
    Layer4
    Layer4_Mask
    Layer3
    Layer3_Mask
    Layer2
    Layer2_Mask
    Layer1=MakeMaterialAttributes!(
     BaseColor=@n1.RGB
     Metallic
     Specular
     Roughness
     Anisotropy
     EmissiveColor=@n29
     Opacity
     OpacityMask
     Normal
     Tangent
     WorldPositionOffset
     SubsurfaceColor
     ClearCoat
     ClearCoatRoughness
     AmbientOcclusion
     Refraction
     CustomizedUVs_0
     CustomizedUVs_1
     CustomizedUVs_2
     CustomizedUVs_3
     CustomizedUVs_4
     CustomizedUVs_5
     CustomizedUVs_6
     CustomizedUVs_7
     PixelDepthOffset
     ShadingModel
     Displacement
    )
    Layer1_Mask=Add(A=@n30, B=@n4.G)
    Background=MakeMaterialAttributes(BaseColor=@n31.RGB, Metallic=@n32.G, Specular=@n32.G, Roughness=@n32.G)
    "<BakedNormal>"=TextureSample() {samplerType=SAMPLERTYPE_Normal, texture=/BRCosmetics/Weapons/FORT_Melee/Pickaxe_Fortnite_DJ/Textures/T_Pickaxe_Fortnite_DJ_N}.RGB
   ) {materialFunction=/Engine/Functions/MaterialLayerFunctions/MatLayerBlend_TenLayerBlend}.Blended Material) {materialFunction=&MF_TechArtOverrides_accessories}
 ]
}
