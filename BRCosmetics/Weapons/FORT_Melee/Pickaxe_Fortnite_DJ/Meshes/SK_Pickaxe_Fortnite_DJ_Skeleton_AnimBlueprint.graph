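# Anim Blueprint for the DJ pickaxe skeleton that adds procedural secondary motion to its three sliders and its handle and head cords.
# The input pose runs through chained Anim Dynamics nodes with linear springs and gravity, confining sliders to local Z travel and the handle cord base to X.
# Slider 2 gets an extra half-strength, Z-only Spring Controller.
# Three head cord joints simulate relative to their base, the first two adding angular springs; event logic is empty.
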AnimBlueprint /BRCosmetics/Weapons/FORT_Melee/Pickaxe_Fortnite_DJ/Meshes/SK_Pickaxe_Fortnite_DJ_Skeleton_AnimBlueprint
target_skeleton=&/SK_Pickaxe_Fortnite_DJ_Skeleton
graphs [
 AnimGraph {nodes [
   AnimGraphNode_SubInput_0 = Input Pose()
   n1 = Anim Dynamics - Bone: dyn_slider1End(ComponentPose=Local To Component(LocalPose=@AnimGraphNode_SubInput_0), Alpha=1, ExternalForce="0,0,0") {
    bGravityOverrideInSimSpace=true
    bLinearSpring=true
    bOverrideLinearDamping=true
    gravityScale=0.9
    linearDampingOverride=1
    linearSpringConstant=850
    alphaBoolBlend {blendOption=Linear}
    boundBone {boneName=dyn_slider1End}
    componentAppliedLinearAccClamp=<1000 1000 1000>
    componentLinearAccScale=<0.8 0.8 0.8>
    componentLinearVelScale=<0.5 0.5 0.5>
    gravityOverride {z=-980}
    physicsBodyDefinitions [{
      sphereCollisionRadius=0
      boundBone {boneName=dyn_slider1End}
      boxExtents=<80 80 5>
      constraintSetup {twistAxis=AxisZ, angularLimitsMax=<0 0 0>, angularLimitsMin=<0 0 0>, angularTarget=<0 0 0>, linearAxesMax=<0 0 1>, linearAxesMin=<0 0 -5>}
      localJointOffset=<-0.1 0 0>
     }]
    relativeSpaceBone {boneName=dyn_slider1Start}
   }
   n2 = Anim Dynamics - Bone: dyn_slider2End(ComponentPose=@n1, Alpha=1, ExternalForce="0,0,0") {
    bGravityOverrideInSimSpace=true
    bLinearSpring=true
    bOverrideLinearDamping=true
    gravityScale=0.9
    linearDampingOverride=1
    linearSpringConstant=500
    alphaBoolBlend {blendOption=Linear}
    boundBone {boneName=dyn_slider2End}
    componentAppliedLinearAccClamp=<1000 1000 1000>
    componentLinearAccScale=<0.8 0.8 0.8>
    componentLinearVelScale=<0.5 0.5 0.5>
    gravityOverride {z=-980}
    physicsBodyDefinitions [{
      sphereCollisionRadius=0
      boundBone {boneName=dyn_slider2End}
      boxExtents=<30 30 4>
      constraintSetup {twistAxis=AxisZ, angularLimitsMax=<0 0 0>, angularLimitsMin=<0 0 0>, angularTarget=<0 0 0>, linearAxesMax=<0 0 1>, linearAxesMin=<0 0 -7>}
      localJointOffset=<-0.1 0 0>
     }]
    relativeSpaceBone {boneName=dyn_slider2Start}
   }
   n3 = Anim Dynamics - Bone: dyn_slider3End(ComponentPose=@n2, Alpha=1, ExternalForce="0,0,0") {
    bGravityOverrideInSimSpace=true
    bLinearSpring=true
    bOverrideLinearDamping=true
    gravityScale=0.9
    linearDampingOverride=0.9999
    linearSpringConstant=600
    alphaBoolBlend {blendOption=Linear}
    boundBone {boneName=dyn_slider3End}
    componentAppliedLinearAccClamp=<1000 1000 1000>
    componentLinearAccScale=<1 1 1>
    componentLinearVelScale=<0.6 0.6 0.6>
    gravityOverride {z=-980}
    physicsBodyDefinitions [{
      sphereCollisionRadius=0
      boundBone {boneName=dyn_slider3End}
      boxExtents=<40 40 4>
      constraintSetup {twistAxis=AxisZ, angularLimitsMax=<0 0 0>, angularLimitsMin=<0 0 0>, angularTarget=<0 0 0>, linearAxesMax=<0 0 1>, linearAxesMin=<0 0 -5>}
      localJointOffset=<-0.1 0 0>
     }]
    relativeSpaceBone {boneName=dyn_slider3Start}
   }
   n4 = Spring controller - Bone: dyn_slider2End!(ComponentPose:Component Space Pose Link Structure=@n3) -> (Pose:Component Space Pose Link Structure) {
    is=title AnimGraphNode_SpringBone
    alpha=0.5
    bLimitDisplacement=true
    bTranslateX=false
    bTranslateY=false
    maxDisplacement=0.5
    springDamping=20
    springStiffness=1000
    alphaBoolBlend {blendOption=Linear}
    springBone {boneName=dyn_slider2End}
   }
   n5 = Anim Dynamics - Bone: dyn_Handle_Cord_Base(ComponentPose=@n4, Alpha=1, ExternalForce="0,0,0") {
    bGravityOverrideInSimSpace=true
    bLinearSpring=true
    bOverrideLinearDamping=true
    gravityScale=0.9
    linearDampingOverride=0.985
    linearSpringConstant=700
    alphaBoolBlend {blendOption=Linear}
    boundBone {boneName=dyn_Handle_Cord_Base}
    componentAppliedLinearAccClamp=<1000 1000 1000>
    componentLinearAccScale=<0.2 0.2 0.2>
    componentLinearVelScale=<0.1 0.1 0.1>
    gravityOverride {z=-980}
    physicsBodyDefinitions [{
      sphereCollisionRadius=0
      boundBone {boneName=dyn_Handle_Cord_Base}
      boxExtents=<4 4 4>
      constraintSetup {angularLimitsMax=<0 0 0>, angularLimitsMin=<0 0 0>, angularTarget=<0 0 0>, linearAxesMax=<5 0 0>, linearAxesMin=<-10 0 0>}
      localJointOffset=<-0.1 0 0>
     }]
    relativeSpaceBone {boneName=Melee_TwoHanded_Handle}
   }
   n6 = Anim Dynamics - Bone: dyn_Head_Cord_jnt1(ComponentPose=@n5, Alpha=1, ExternalForce="0,0,0") {
    angularDampingOverride=0.966
    angularSpringConstant=700
    bAngularSpring=true
    bGravityOverrideInSimSpace=true
    bLinearSpring=true
    bOverrideAngularDamping=true
    bOverrideLinearDamping=true
    gravityScale=0.9
    linearDampingOverride=0.975
    linearSpringConstant=700
    alphaBoolBlend {blendOption=Linear}
    boundBone {boneName=dyn_Head_Cord_jnt1}
    componentAppliedLinearAccClamp=<1000 1000 1000>
    componentLinearAccScale=<0.6 0.6 0.6>
    componentLinearVelScale=<0.15 0.15 0.15>
    gravityOverride {z=-980}
    physicsBodyDefinitions [{
      sphereCollisionRadius=0
      boundBone {boneName=dyn_Head_Cord_jnt1}
      boxExtents=<7 7 7>
      constraintSetup {
       twistAxis=AxisZ
       angularLimitsMax=<10 10 10>
       angularLimitsMin=<-10 -10 -10>
       angularTarget=<0 0 0>
       linearAxesMax=<2 2 2>
       linearAxesMin=<-2 -2 -2>
      }
      localJointOffset=<-0.4 0 -0.25>
     }]
    relativeSpaceBone {boneName=dyn_Head_Cord_Base}
   }
   n7 = Anim Dynamics - Bone: dyn_Head_Cord_jnt2(ComponentPose=@n6, Alpha=1, ExternalForce="0,0,0") {
    angularDampingOverride=0.985
    angularSpringConstant=800
    bAngularSpring=true
    bGravityOverrideInSimSpace=true
    bLinearSpring=true
    bOverrideAngularDamping=true
    bOverrideLinearDamping=true
    gravityScale=0.9
    linearDampingOverride=0.985
    linearSpringConstant=700
    alphaBoolBlend {blendOption=Linear}
    boundBone {boneName=dyn_Head_Cord_jnt2}
    componentAppliedLinearAccClamp=<1000 1000 1000>
    componentLinearAccScale=<0.6 0.6 0.6>
    componentLinearVelScale=<0.1 0.1 0.1>
    gravityOverride {z=-980}
    physicsBodyDefinitions [{
      sphereCollisionRadius=0
      boundBone {boneName=dyn_Head_Cord_jnt2}
      boxExtents=<4 4 4>
      constraintSetup {
       twistAxis=AxisZ
       angularLimitsMax=<10 10 10>
       angularLimitsMin=<-10 -10 -10>
       angularTarget=<0 0 0>
       linearAxesMax=<2 2 2>
       linearAxesMin=<-3 -3 -3>
      }
      localJointOffset=<-0.4 -0.2 -0.25>
     }]
    relativeSpaceBone {boneName=dyn_Head_Cord_Base}
   }
   n8 = Anim Dynamics - Bone: dyn_Head_Cord_jnt3(ComponentPose=@n7, Alpha=1, ExternalForce="0,0,0") {
    angularDampingOverride=0.985
    angularSpringConstant=5000
    bGravityOverrideInSimSpace=true
    bLinearSpring=true
    bOverrideLinearDamping=true
    gravityScale=0.9
    linearDampingOverride=0.985
    linearSpringConstant=400
    alphaBoolBlend {blendOption=Linear}
    boundBone {boneName=dyn_Head_Cord_jnt3}
    componentAppliedLinearAccClamp=<1000 1000 1000>
    componentLinearAccScale=<0.8 0.8 0.8>
    componentLinearVelScale=<0.2 0.2 0.2>
    gravityOverride {z=-980}
    physicsBodyDefinitions [{
      sphereCollisionRadius=0
      boundBone {boneName=dyn_Head_Cord_jnt3}
      boxExtents=<15 18 18>
      constraintSetup {
       angularTargetAxis=AxisZ
       twistAxis=AxisZ
       angularLimitsMax=<0 5 0>
       angularLimitsMin=<0 -5 0>
       angularTarget=<0 0 0>
       linearAxesMax=<3 3 3>
       linearAxesMin=<-2 -2 -2>
      }
      localJointOffset=<-1 0 -0.3>
     }]
    relativeSpaceBone {boneName=dyn_Head_Cord_Base}
   }
   Output Pose(Result=Component To Local(ComponentPose=@n8)) {layerGroup=None}
  ]}
 EventGraph {nodes [Event BlueprintUpdateAnimation()]}
]
variables []
